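; Passive gameplay ability for the Season 15 Battle Pass "give Valentine" quest, applying a Valentine GameplayEffect to the owner while that objective is in progress.
; On activation it binds to the controller's quest objective state event and checks a hardcoded Week 11 quest objective before applying the effect.
; When an objective completes on the configured Quest variable, it removes effects granting the GiveValentine tag, resets its handle variable and unbinds the listener.
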(event Ability|EventActivateAbility
  (|CallParentFunction)
  (EventDispatchers|BindEventtoOnQuestObjectiveStateChanged (Fortnite|GetFortPlayerControllerfromActor (Variables|GAVars|Data|GetOwningPlayerPawn)) (delegate Custom|QuestComplete))
  (if (Quest|IsObjectivewithNameinProgress (Variables|GAVars|Data|GetOwningPlayerQuestManager) "/BattlepassS15/Items/QuestItems/Quest_S15_W11_VR_Q07.Quest_S15_W11_VR_Q07" "Quest_S15_W11_VR_Q07_obj0" :!class /Script/FortniteGame.FortQuestManager)
    (GameplayEffects|ApplyGameplayEffectToSelf :self (Ability|GetAbilitySystemComponent (Variables|GAVars|Data|GetOwningPlayerPawn)) :GameplayEffectClass (Variables|GAVars|Setup|GetGE_Valentine) :Level 0.0)))

(event Ability|EventFailedToActivatePassiveAbility
  (params (ActorInfo :type Gameplay%20Ability%20Actor%20Info%20Structure)))

(event Custom|QuestComplete
  (params (Controller :type Fort%20Player%20Controller%20Object%20Reference) (QuestDef :type Fort%20Quest%20Item%20Definition%20Object%20Reference) (BackendName :type Name) (CompletionCount :type Integer) (ObjectiveCompleted :type Boolean) (QuestCompleted :type Boolean))
  (if (Utilities|Equal(Object) QuestDef (Variables|GAVars|Setup|GetQuest))
    (if ObjectiveCompleted
      (GameplayEffects|RemoveActiveEffectswithGrantedTags (Ability|GetAbilitySystemComponent (Variables|GAVars|Data|GetOwningPlayerPawn)) "(GameplayTags=((TagName=\"Athena.Quests.Crush.GiveValentine\")))")
      (Variables|GAVars|Data|SetAwakening%20Handle)
      (EventDispatchers|UnbindEventfromOnQuestObjectiveStateChanged (Fortnite|GetFortPlayerControllerfromActor (Variables|GAVars|Data|GetOwningPlayerPawn)) (delegate Custom|QuestComplete)))))
